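# Physics asset for the female medium Dino Hunter Armor cosmetic body, providing bodies for secondary motion on its skirt and sword.
# Front, side and back skirt panels on each side form limited base-to-A-to-B chains, ending in boxes, with tighter lower-joint limits on side panels.
# The sword capsule has narrow limits to its base, which joins the left thigh with free rotation.
# Capsules on both thighs supply leg volumes.
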PhysicsAsset /BRCosmetics/Characters/Player/Female/Medium/Bodies/F_MED_Dino_Hunter_Armor/Meshes/F_MED_Dino_Hunter_Armor_Physics
bodies {
 dyn_skirt_bk_a_base_l [{shapeType=Capsule, length=0.505, radius=0.505, rotation=<rot 0 -90 0>}]
 dyn_skirt_bk_a_base_r [{shapeType=Capsule, length=0.505, radius=0.505, rotation=<rot 0 -90 0>}]
 dyn_skirt_bk_a_l [{shapeType=Capsule, length=3.75, radius=2.25, center=<4 0 0>, rotation=<rot 90 0 0>}]
 dyn_skirt_bk_a_r [{shapeType=Capsule, length=3.75, radius=2.25, center=<-4 0 0>, rotation=<rot 90 0 0>}]
 dyn_skirt_bk_b_l [{shapeType=Box, extentX=16.5, extentY=6.5, extentZ=28.5, center=<8.5058 0.0001 0>}]
 dyn_skirt_bk_b_r [{shapeType=Box, extentX=18.5, extentY=9.5, extentZ=26.5, center=<-9.7755 0 -0.0001>}]
 dyn_skirt_fr_a_base_l [{shapeType=Capsule, length=0.505, radius=0.505, rotation=<rot 0 -90 0>}]
 dyn_skirt_fr_a_base_r [{shapeType=Capsule, length=0.505, radius=0.505, rotation=<rot 0 -90 0>}]
 dyn_skirt_fr_a_l [{shapeType=Capsule, length=3.75, radius=2.25, center=<4 0 0>, rotation=<rot 90 0 0>}]
 dyn_skirt_fr_a_r [{shapeType=Capsule, length=3.75, radius=2.25, center=<-4 0 0>, rotation=<rot 90 0 0>}]
 dyn_skirt_fr_b_l [{shapeType=Box, extentX=14.5, extentY=5, extentZ=21.5, center=<7.9955 0 0>}]
 dyn_skirt_fr_b_r [{shapeType=Box, extentX=19, extentY=6.5, extentZ=25.5, center=<-9.3403 0.0003 0>}]
 dyn_skirt_side_a_base_l [{shapeType=Capsule, length=0.505, radius=0.505, rotation=<rot 0 -90 0>}]
 dyn_skirt_side_a_base_r [{shapeType=Capsule, length=0.505, radius=0.505, rotation=<rot 0 -90 0>}]
 dyn_skirt_side_a_l [{shapeType=Capsule, length=3.75, radius=2.25, center=<4 0 0>, rotation=<rot 90 0 0>}]
 dyn_skirt_side_a_r [{shapeType=Capsule, length=3.75, radius=2.25, center=<-4 0 0>, rotation=<rot 90 0 0>}]
 dyn_skirt_side_b_l [{shapeType=Box, extentX=16.5, extentY=3.5, extentZ=22, center=<8.467 0 0>}]
 dyn_skirt_side_b_r [{shapeType=Box, extentX=18.5, extentY=4.5, extentZ=26, center=<-9.7755 0 -0.0001>}]
 dyn_sword [{shapeType=Capsule, length=20.8263, radius=5.5621, center=<2 0 0>, rotation=<rot 90 0 0>}]
 dyn_sword_base [{shapeType=Capsule, length=0.505, radius=0.505, rotation=<rot 0 -90 0>}]
 thigh_l [{shapeType=Capsule, length=34, radius=6.5, center=<-23.8158 0 0.2062>, rotation=<rot 89.5039 0.0011 0.0011>}]
 thigh_r [{shapeType=Capsule, length=34, radius=6.5, center=<23.8158 0 -0.2062>, rotation=<rot 89.5039 0.0011 -179.9988>}]
}
constraints [
 {bone1Name=dyn_skirt_fr_a_l, bone2Name=dyn_skirt_fr_a_base_l, swing1LimitDegrees=20, swing2LimitDegrees=10, twistLimitDegrees=5}
 {bone1Name=dyn_skirt_fr_b_l, bone2Name=dyn_skirt_fr_a_l, swing1LimitDegrees=30, swing2LimitDegrees=15, twistLimitDegrees=15}
 {bone1Name=dyn_skirt_side_a_l, bone2Name=dyn_skirt_side_a_base_l, swing1LimitDegrees=20, swing2LimitDegrees=10, twistLimitDegrees=5}
 {bone1Name=dyn_skirt_side_b_l, bone2Name=dyn_skirt_side_a_l, swing1LimitDegrees=5, swing2LimitDegrees=5, twistLimitDegrees=5}
 {bone1Name=dyn_skirt_bk_a_l, bone2Name=dyn_skirt_bk_a_base_l, swing1LimitDegrees=20, swing2LimitDegrees=10, twistLimitDegrees=5}
 {bone1Name=dyn_skirt_bk_b_l, bone2Name=dyn_skirt_bk_a_l, swing1LimitDegrees=30, swing2LimitDegrees=15, twistLimitDegrees=15}
 {bone1Name=dyn_skirt_fr_a_r, bone2Name=dyn_skirt_fr_a_base_r, swing1LimitDegrees=20, swing2LimitDegrees=10, twistLimitDegrees=5}
 {bone1Name=dyn_skirt_fr_b_r, bone2Name=dyn_skirt_fr_a_r, swing1LimitDegrees=30, swing2LimitDegrees=15, twistLimitDegrees=15}
 {bone1Name=dyn_skirt_side_a_r, bone2Name=dyn_skirt_side_a_base_r, swing1LimitDegrees=20, swing2LimitDegrees=10, twistLimitDegrees=5}
 {bone1Name=dyn_skirt_side_b_r, bone2Name=dyn_skirt_side_a_r, swing1LimitDegrees=10, swing2LimitDegrees=10, twistLimitDegrees=10}
 {bone1Name=dyn_skirt_bk_a_r, bone2Name=dyn_skirt_bk_a_base_r, swing1LimitDegrees=20, swing2LimitDegrees=10, twistLimitDegrees=5}
 {bone1Name=dyn_skirt_bk_b_r, bone2Name=dyn_skirt_bk_a_r, swing1LimitDegrees=30, swing2LimitDegrees=15, twistLimitDegrees=15}
 {
  bone1Name=dyn_sword_base
  bone2Name=thigh_l
  swing1LimitDegrees=45
  swing1Motion=Free
  swing2LimitDegrees=45
  swing2Motion=Free
  twistLimitDegrees=45
  twistMotion=Free
 }
 {bone1Name=dyn_sword, bone2Name=dyn_sword_base, swing1LimitDegrees=10, swing2LimitDegrees=5, twistLimitDegrees=2}
]
