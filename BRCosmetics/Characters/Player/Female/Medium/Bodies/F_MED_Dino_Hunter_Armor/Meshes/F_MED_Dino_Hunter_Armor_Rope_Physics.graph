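# Physics asset for the Dino Hunter armor on the medium female body, covering the dangling rope and front-right pouch.
# The rope chains from a hook base through a nearly rigid hook into two box segments with limited swing and twist.
# The pouch hangs from its own base with tight angular limits.
# Right thigh and calf capsules, joined by a free-rotating constraint, give the leg collision shapes alongside these dangling parts.
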PhysicsAsset /BRCosmetics/Characters/Player/Female/Medium/Bodies/F_MED_Dino_Hunter_Armor/Meshes/F_MED_Dino_Hunter_Armor_Rope_Physics
bodies {
 calf_r [{shapeType=Capsule, length=37.25, radius=6.75, center=<20.151 0.0002 0>, rotation=<rot 90 0 0>}]
 dyn_armor_pouch_fr_base_r [{shapeType=Capsule, length=0.505, radius=0.505, rotation=<rot 0 -90 0>}]
 dyn_armor_pouch_fr_r [{shapeType=Capsule, length=10.5, radius=3.5, center=<1.2298 0 0>}]
 dyn_armor_rope_a [{shapeType=Box, extentX=11, extentY=6.5, extentZ=35.5, center=<5.5 0 0>}]
 dyn_armor_rope_b [{shapeType=Box, extentX=20, extentY=6.5, extentZ=35.5, center=<10.2432 0 0>}]
 dyn_armor_rope_hook [{shapeType=Capsule, length=2, radius=3, center=<2.4177 0 0>, rotation=<rot 90 0 0>}]
 dyn_armor_rope_hook_base [{shapeType=Capsule, length=0.505, radius=0.505, rotation=<rot 0 -90 0>}]
 thigh_r [{shapeType=Capsule, length=34, radius=6.5, center=<23.8158 0 -0.2062>, rotation=<rot 89.5039 0.0011 -179.9988>}]
}
constraints [
 {
  bone1Name=dyn_armor_rope_hook
  bone2Name=dyn_armor_rope_hook_base
  swing1LimitDegrees=2
  swing2LimitDegrees=2
  twistLimitDegrees=5
  twistMotion=Locked
 }
 {bone1Name=dyn_armor_rope_a, bone2Name=dyn_armor_rope_hook, swing1LimitDegrees=20, swing2LimitDegrees=20, twistLimitDegrees=20}
 {bone1Name=dyn_armor_rope_b, bone2Name=dyn_armor_rope_a, swing1LimitDegrees=20, swing2LimitDegrees=20, twistLimitDegrees=20}
 {
  bone1Name=calf_r
  bone2Name=thigh_r
  swing1LimitDegrees=45
  swing1Motion=Free
  swing2LimitDegrees=45
  swing2Motion=Free
  twistLimitDegrees=45
  twistMotion=Free
 }
 {bone1Name=dyn_armor_pouch_fr_r, bone2Name=dyn_armor_pouch_fr_base_r, swing1LimitDegrees=10, swing2LimitDegrees=5, twistLimitDegrees=5}
]
