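# Physics asset for the shirt pieces of the female medium Dino Hunter Armor body, giving its dyn_shirts bones collision and joints.
# Five two-body chains (front center, left and right; back center and left) hang from a spine_03 capsule that represents the torso.
# Tiny base capsules attach to spine_03 with free swing and twist, while the larger outer capsules join their bases with narrow, per-piece angular limits.
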PhysicsAsset /BRCosmetics/Characters/Player/Female/Medium/Bodies/F_MED_Dino_Hunter_Armor/Meshes/F_MED_Dino_Hunter_Armor_Shirts_Physics
bodies {
 dyn_shirts_bk [{shapeType=Capsule, length=3.75, radius=1.5, center=<3 0 0>, rotation=<rot 90 0 0>}]
 dyn_shirts_bk_base [{shapeType=Capsule, length=0.505, radius=0.505, rotation=<rot 0 -90 0>}]
 dyn_shirts_bk_base_l [{shapeType=Capsule, length=0.505, radius=0.505, rotation=<rot 0 -90 0>}]
 dyn_shirts_bk_l [{shapeType=Capsule, length=3.5, radius=1.75, center=<3 0 0>, rotation=<rot 90 0 0>}]
 dyn_shirts_fr [{shapeType=Capsule, length=4, radius=1.25, center=<3 0 0>, rotation=<rot 90 0 0>}]
 dyn_shirts_fr_base [{shapeType=Capsule, length=0.505, radius=0.505, rotation=<rot 0 -90 0>}]
 dyn_shirts_fr_base_l [{shapeType=Capsule, length=0.505, radius=0.505, rotation=<rot 0 -90 0>}]
 dyn_shirts_fr_base_r [{shapeType=Capsule, length=0.505, radius=0.505, rotation=<rot 0 -90 0>}]
 dyn_shirts_fr_l [{shapeType=Capsule, length=4, radius=1.25, center=<3 0 0>, rotation=<rot 90 0 0>}]
 dyn_shirts_fr_r [{shapeType=Capsule, length=4, radius=1.25, center=<3 0 0>, rotation=<rot 90 0 0>}]
 spine_03 [{shapeType=Capsule, length=2.5, radius=7, center=<-0.7665 2.2902 0>, rotation=<rot 90 0 0>}]
}
constraints [
 {
  bone1Name=dyn_shirts_fr_base
  bone2Name=spine_03
  swing1LimitDegrees=45
  swing1Motion=Free
  swing2LimitDegrees=45
  swing2Motion=Free
  twistLimitDegrees=45
  twistMotion=Free
 }
 {
  bone1Name=dyn_shirts_fr_base_l
  bone2Name=spine_03
  swing1LimitDegrees=45
  swing1Motion=Free
  swing2LimitDegrees=45
  swing2Motion=Free
  twistLimitDegrees=45
  twistMotion=Free
 }
 {
  bone1Name=dyn_shirts_fr_base_r
  bone2Name=spine_03
  swing1LimitDegrees=45
  swing1Motion=Free
  swing2LimitDegrees=45
  swing2Motion=Free
  twistLimitDegrees=45
  twistMotion=Free
 }
 {
  bone1Name=dyn_shirts_bk_base_l
  bone2Name=spine_03
  swing1LimitDegrees=45
  swing1Motion=Free
  swing2LimitDegrees=45
  swing2Motion=Free
  twistLimitDegrees=45
  twistMotion=Free
 }
 {
  bone1Name=dyn_shirts_bk_base
  bone2Name=spine_03
  swing1LimitDegrees=45
  swing1Motion=Free
  swing2LimitDegrees=45
  swing2Motion=Free
  twistLimitDegrees=45
  twistMotion=Free
 }
 {bone1Name=dyn_shirts_fr, bone2Name=dyn_shirts_fr_base, swing1LimitDegrees=10, swing2LimitDegrees=3, twistLimitDegrees=3}
 {bone1Name=dyn_shirts_fr_l, bone2Name=dyn_shirts_fr_base_l, swing1LimitDegrees=3, swing2LimitDegrees=3, twistLimitDegrees=3}
 {bone1Name=dyn_shirts_fr_r, bone2Name=dyn_shirts_fr_base_r, swing1LimitDegrees=5, swing2LimitDegrees=5, twistLimitDegrees=5}
 {bone1Name=dyn_shirts_bk_l, bone2Name=dyn_shirts_bk_base_l, swing1LimitDegrees=5, swing2LimitDegrees=5, twistLimitDegrees=5}
 {bone1Name=dyn_shirts_bk, bone2Name=dyn_shirts_bk_base, swing1LimitDegrees=3, swing2LimitDegrees=3, twistLimitDegrees=3}
]
