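# Animation Blueprint for the female medium Dino Hunter Armor body cosmetic, adding secondary motion to its armor, skirt, shirt and accessories.
# It copies the attached parent mesh's pose, then places helper bones with weighted constraints.
# Thigh, elbow and upper-arm rotation variables scale additive bone corrections.
# Three RigidBody nodes with separate physics assets and per-bone spring AnimDynamics, for sword teeth and shoestrings, follow in the chain.
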AnimBlueprint /BRCosmetics/Characters/Player/Female/Medium/Bodies/F_MED_Dino_Hunter_Armor/Meshes/F_MED_Dino_Hunter_Armor_AnimBP
target_skeleton=&/F_MED_Dino_Hunter_Armor_Skeleton
graphs [
 AnimGraph {nodes [
   n1 = "Constraint - dyn_skirt_bits_a_base (PPPPP)"!(
    ComponentPose
    Alpha=1
    ConstraintWeights_0=0.0199
    ConstraintWeights_1:float=0.0366
    ConstraintWeights_2:float=0.0655
    ConstraintWeights_3:float=0.7585
    ConstraintWeights_4:float=0.1194
   ) -> (Pose) {
    alphaBoolBlend {blendOption=Linear}
    boneToModify {boneName=dyn_skirt_bits_a_base}
    constraintSetup [
     {targetBone {boneName=dyn_skirt_side_b_l}}
     {targetBone {boneName=dyn_skirt_fr_b_l}}
     {targetBone {boneName=dyn_skirt_fr_a_r}}
     {targetBone {boneName=dyn_skirt_fr_a_l}}
     {targetBone {boneName=dyn_skirt_side_a_l}}
    ]
    constraintWeights [1,1,1,1,1]
   }
   n2 = "Constraint - dyn_skirt_bits_b_base (PPPPPP)"!(
    ComponentPose=@n1
    Alpha=1
    ConstraintWeights_0=0.0372
    ConstraintWeights_1:float=0.0548
    ConstraintWeights_2:float=0.0307
    ConstraintWeights_3:float=0.531
    ConstraintWeights_4:float=0.0128
    ConstraintWeights_5:float=0.3334
   ) -> (Pose) {
    alphaBoolBlend {blendOption=Linear}
    boneToModify {boneName=dyn_skirt_bits_b_base}
    constraintSetup [
     {targetBone {boneName=dyn_skirt_bk_a_l}}
     {targetBone {boneName=dyn_skirt_side_b_l}}
     {targetBone {boneName=dyn_skirt_fr_b_l}}
     {targetBone {boneName=dyn_skirt_fr_a_l}}
     {targetBone {boneName=dyn_skirt_fr_a_r}}
     {targetBone {boneName=dyn_skirt_side_a_l}}
    ]
    constraintWeights [1,1,1,1,1,1]
   }
   n3 = "Constraint - dyn_skirt_bits_c_base (PPPP)"!(
    ComponentPose=@n2
    Alpha=1
    ConstraintWeights_0=0.3275
    ConstraintWeights_1:float=0.0751
    ConstraintWeights_2:float=0.09
    ConstraintWeights_3:float=0.5074
   ) -> (Pose) {
    alphaBoolBlend {blendOption=Linear}
    boneToModify {boneName=dyn_skirt_bits_c_base}
    constraintSetup [
     {targetBone {boneName=dyn_skirt_bk_a_l}}
     {targetBone {boneName=dyn_skirt_side_b_l}}
     {targetBone {boneName=dyn_skirt_fr_a_l}}
     {targetBone {boneName=dyn_skirt_side_a_l}}
    ]
    constraintWeights [1,1,1,1]
   }
   "Constraint - dyn_skirt_bits_d_base (PPPPPPP)"!(
    ComponentPose=@n3
    Alpha=1
    ConstraintWeights_0=0.552
    ConstraintWeights_1:float=0.0131
    ConstraintWeights_2:float=0.0113
    ConstraintWeights_3:float=0.057
    ConstraintWeights_4:float=0.0287
    ConstraintWeights_5:float=0.0685
    ConstraintWeights_6:float=0.2695
   ) -> (Pose) {
    alphaBoolBlend {blendOption=Linear}
    boneToModify {boneName=dyn_skirt_bits_d_base}
    constraintSetup [
     {targetBone {boneName=dyn_skirt_bk_a_l}}
     {targetBone {boneName=dyn_skirt_side_b_l}}
     {targetBone {boneName=dyn_skirt_side_a_r}}
     {targetBone {boneName=dyn_skirt_bk_b_l}}
     {targetBone {boneName=dyn_skirt_bk_b_r}}
     {targetBone {boneName=dyn_skirt_side_a_l}}
     {targetBone {boneName=dyn_skirt_bk_a_r}}
    ]
    constraintWeights [1,1,1,1,1,1,1]
   }
   n4 = "Transform (Modify) Bone - Bone: dyn_skirt_bk_a_base_l"(Alpha=1, Translation="0,1,0", Rotation="0,0,0", Scale="1,1,1") {
    rotationMode=BMM_Additive
    rotationSpace=BCS_BoneSpace
    scaleMode=BMM_Additive
    scaleSpace=BCS_BoneSpace
    translationMode=BMM_Additive
    translationSpace=BCS_BoneSpace
    alphaBoolBlend {blendOption=Linear}
    alphaScaleBiasClamp {inRange {max=0, min=-100}, outRange {max=0, min=1}}
    boneToModify {boneName=dyn_skirt_bk_a_base_l}
    translation {y=1}
   }
   n5 = "Transform (Modify) Bone - Bone: dyn_skirt_bk_a_base_r"(ComponentPose=@n4, Alpha=1, Translation="0,-1,0", Rotation="0,0,0", Scale="1,1,1") {
    rotationMode=BMM_Additive
    rotationSpace=BCS_BoneSpace
    scaleMode=BMM_Additive
    scaleSpace=BCS_BoneSpace
    translationMode=BMM_Additive
    translationSpace=BCS_BoneSpace
    alphaBoolBlend {blendOption=Linear}
    alphaScaleBiasClamp {inRange {max=0, min=-100}, outRange {max=0, min=1}}
    boneToModify {boneName=dyn_skirt_bk_a_base_r}
    translation {y=-1}
   }
   n6 = "Transform (Modify) Bone - Bone: dyn_skirt_fr_a_base_r"(ComponentPose=@n5, Alpha=1, Translation="0,-1,0", Rotation="0,0,0", Scale="1,1,1") {
    rotationMode=BMM_Additive
    rotationSpace=BCS_BoneSpace
    scaleMode=BMM_Additive
    scaleSpace=BCS_BoneSpace
    translationMode=BMM_Additive
    translationSpace=BCS_BoneSpace
    alphaBoolBlend {blendOption=Linear}
    alphaScaleBiasClamp {inRange {max=0, min=-100}, outRange {max=0, min=1}}
    boneToModify {boneName=dyn_skirt_fr_a_base_r}
    translation {y=-1}
   }
   "Transform (Modify) Bone - Bone: dyn_skirt_side_a_base_r"(ComponentPose=@n6, Alpha=1, Translation="0,-1,0", Rotation="0,0,0", Scale="1,1,1") {
    rotationMode=BMM_Additive
    rotationSpace=BCS_BoneSpace
    scaleMode=BMM_Additive
    scaleSpace=BCS_BoneSpace
    translationMode=BMM_Additive
    translationSpace=BCS_BoneSpace
    alphaBoolBlend {blendOption=Linear}
    alphaScaleBiasClamp {inRange {max=0, min=-100}, outRange {max=0, min=1}}
    boneToModify {boneName=dyn_skirt_side_a_base_r}
    translation {y=-1}
   }
   n7 = "Constraint - dyn_volume_base_lowerarm_l (PP)"!(
    ComponentPose=Local To Component(LocalPose=Copy Pose From Mesh() {bUseAttachedParent=true})
    Alpha=1
    ConstraintWeights_0=0.65
    ConstraintWeights_1:float=0.35
   ) -> (Pose) {
    alphaBoolBlend {blendOption=Linear}
    boneToModify {boneName=dyn_volume_base_lowerarm_l}
    constraintSetup [{targetBone {boneName=upperarm_l}}, {targetBone {boneName=lowerarm_l}}]
    constraintWeights [1,1]
   }
   n8 = "Constraint - dyn_volume_base_lowerarm_r (PP)"!(ComponentPose=@n7, Alpha=1, ConstraintWeights_0=0.65, ConstraintWeights_1:float=0.35) -> (Pose) {
    alphaBoolBlend {blendOption=Linear}
    boneToModify {boneName=dyn_volume_base_lowerarm_r}
    constraintSetup [{targetBone {boneName=upperarm_r}}, {targetBone {boneName=lowerarm_r}}]
    constraintWeights [1,1]
   }
   n9 = "Transform (Modify) Bone - Bone: dyn_volume_lowerarm_l"(ComponentPose=@n8, Alpha=Get elbow_rot_l() -> (elbow_rot_l:double), Translation="0,-2,0", Rotation="0,0,0", Scale="0.3,1,1") {
    rotationMode=BMM_Additive
    rotationSpace=BCS_BoneSpace
    scaleMode=BMM_Additive
    scaleSpace=BCS_BoneSpace
    translationMode=BMM_Additive
    translationSpace=BCS_BoneSpace
    alphaBoolBlend {blendOption=Linear}
    alphaScaleBiasClamp {bMapRange=true, inRange {max=80}}
    boneToModify {boneName=dyn_volume_lowerarm_l}
    scale {x=0.3}
    translation {y=-2}
   }
   n10 = "Transform (Modify) Bone - Bone: dyn_volume_lowerarm_r"(ComponentPose=@n9, Alpha=Get elbow_rot_r() -> (elbow_rot_r:double), Translation="0,2,0", Rotation="0,0,0", Scale="0.3,1,1") {
    rotationMode=BMM_Additive
    rotationSpace=BCS_BoneSpace
    scaleMode=BMM_Additive
    scaleSpace=BCS_BoneSpace
    translationMode=BMM_Additive
    translationSpace=BCS_BoneSpace
    alphaBoolBlend {blendOption=Linear}
    alphaScaleBiasClamp {bMapRange=true, inRange {max=80}}
    boneToModify {boneName=dyn_volume_lowerarm_r}
    scale {x=0.3}
    translation {y=2}
   }
   n11 = "Constraint - dyn_shldrCapHorn_a_l (PPPP)"!(
    ComponentPose=@n10
    Alpha=1
    ConstraintWeights_0=0.21
    ConstraintWeights_1:float=0.1616
    ConstraintWeights_2:float=0.6284
    ConstraintWeights_3:float=0
   ) -> (Pose) {
    alphaBoolBlend {blendOption=Linear}
    boneToModify {boneName=dyn_shldrCapHorn_a_l}
    constraintSetup [
     {targetBone {boneName=upperarm_twist_01_l}}
     {targetBone {boneName=spine_05}}
     {targetBone {boneName=clavicle_l}}
     {targetBone {boneName=upperarm_l}}
    ]
    constraintWeights [1,1,1,1]
   }
   n12 = "Constraint - dyn_shldrCapHorn_c_l (PPPP)"!(
    ComponentPose=@n11
    Alpha=1
    ConstraintWeights_0=0.168
    ConstraintWeights_1:float=0.1622
    ConstraintWeights_2:float=0.6698
    ConstraintWeights_3:float=0
   ) -> (Pose) {
    alphaBoolBlend {blendOption=Linear}
    boneToModify {boneName=dyn_shldrCapHorn_c_l}
    constraintSetup [
     {targetBone {boneName=upperarm_twist_01_l}}
     {targetBone {boneName=spine_05}}
     {targetBone {boneName=clavicle_l}}
     {targetBone {boneName=upperarm_l}}
    ]
    constraintWeights [1,1,1,1]
   }
   n13 = "Constraint - dyn_shldrCapHorn_d_l (PPPP)"!(
    ComponentPose=@n12
    Alpha=1
    ConstraintWeights_0=0.4975
    ConstraintWeights_1:float=0.0199
    ConstraintWeights_2:float=0.4827
    ConstraintWeights_3:float=0
   ) -> (Pose) {
    alphaBoolBlend {blendOption=Linear}
    boneToModify {boneName=dyn_shldrCapHorn_d_l}
    constraintSetup [
     {targetBone {boneName=upperarm_twist_01_l}}
     {targetBone {boneName=spine_05}}
     {targetBone {boneName=clavicle_l}}
     {targetBone {boneName=upperarm_l}}
    ]
    constraintWeights [1,1,1,1]
   }
   n14 = "Constraint - dyn_shldrCapHorn_b_l (PPPP)"!(
    ComponentPose=@n13
    Alpha=1
    ConstraintWeights_0=0.6145
    ConstraintWeights_1:float=0.0303
    ConstraintWeights_2:float=0.3553
    ConstraintWeights_3:float=0
   ) -> (Pose) {
    alphaBoolBlend {blendOption=Linear}
    boneToModify {boneName=dyn_shldrCapHorn_b_l}
    constraintSetup [
     {targetBone {boneName=upperarm_twist_01_l}}
     {targetBone {boneName=spine_05}}
     {targetBone {boneName=clavicle_l}}
     {targetBone {boneName=upperarm_l}}
    ]
    constraintWeights [1,1,1,1]
   }
   n15 = "Constraint - dyn_shldrCapHorn_f_l (PPPP)"!(
    ComponentPose=@n14
    Alpha=1
    ConstraintWeights_0=0.5155
    ConstraintWeights_1:float=0.0252
    ConstraintWeights_2:float=0.4594
    ConstraintWeights_3:float=0
   ) -> (Pose) {
    alphaBoolBlend {blendOption=Linear}
    boneToModify {boneName=dyn_shldrCapHorn_f_l}
    constraintSetup [
     {targetBone {boneName=upperarm_twist_01_l}}
     {targetBone {boneName=spine_05}}
     {targetBone {boneName=clavicle_l}}
     {targetBone {boneName=upperarm_l}}
    ]
    constraintWeights [1,1,1,1]
   }
   n16 = "Constraint - dyn_shldrCapHorn_e_l (PPPP)"!(
    ComponentPose=@n15
    Alpha=1
    ConstraintWeights_0=0.135
    ConstraintWeights_1:float=0.1694
    ConstraintWeights_2:float=0.6956
    ConstraintWeights_3:float=0
   ) -> (Pose) {
    alphaBoolBlend {blendOption=Linear}
    boneToModify {boneName=dyn_shldrCapHorn_e_l}
    constraintSetup [
     {targetBone {boneName=upperarm_twist_01_l}}
     {targetBone {boneName=spine_05}}
     {targetBone {boneName=clavicle_l}}
     {targetBone {boneName=upperarm_l}}
    ]
    constraintWeights [1,1,1,1]
   }
   n17 = "Constraint - dyn_shldrCapHorn_g_l (PPP)"!(ComponentPose=@n16, Alpha=1, ConstraintWeights_0=0.7, ConstraintWeights_1:float=0.15, ConstraintWeights_2:float=0.15) -> (Pose) {
    alphaBoolBlend {blendOption=Linear}
    boneToModify {boneName=dyn_shldrCapHorn_g_l}
    constraintSetup [{targetBone {boneName=clavicle_l}}, {targetBone {boneName=spine_05}}, {targetBone {boneName=upperarm_twist_01_l}}]
    constraintWeights [1,1,1]
   }
   n18 = "Constraint - dyn_shldrCapHorn_a_r (PPPP)"!(
    ComponentPose=@n17
    Alpha=1
    ConstraintWeights_0=0.3527
    ConstraintWeights_1:float=0.6182
    ConstraintWeights_2:float=0
    ConstraintWeights_3:float=0.0291
   ) -> (Pose) {
    alphaBoolBlend {blendOption=Linear}
    boneToModify {boneName=dyn_shldrCapHorn_a_r}
    constraintSetup [
     {targetBone {boneName=clavicle_r}}
     {targetBone {boneName=upperarm_twist_01_r}}
     {targetBone {boneName=upperarm_r}}
     {targetBone {boneName=spine_05}}
    ]
    constraintWeights [1,1,1,1]
   }
   n19 = "Constraint - dyn_shldrCapHorn_b_r (PPPP)"!(
    ComponentPose=@n18
    Alpha=1
    ConstraintWeights_0=0.4759
    ConstraintWeights_1:float=0.5062
    ConstraintWeights_2:float=0
    ConstraintWeights_3:float=0.0179
   ) -> (Pose) {
    alphaBoolBlend {blendOption=Linear}
    boneToModify {boneName=dyn_shldrCapHorn_b_r}
    constraintSetup [
     {targetBone {boneName=clavicle_r}}
     {targetBone {boneName=upperarm_twist_01_r}}
     {targetBone {boneName=upperarm_r}}
     {targetBone {boneName=spine_05}}
    ]
    constraintWeights [1,1,1,1]
   }
   n20 = "Constraint - dyn_shldrCapHorn_c_r (PPPP)"!(
    ComponentPose=@n19
    Alpha=1
    ConstraintWeights_0=0.6679
    ConstraintWeights_1:float=0.1987
    ConstraintWeights_2:float=0
    ConstraintWeights_3:float=0.1335
   ) -> (Pose) {
    alphaBoolBlend {blendOption=Linear}
    boneToModify {boneName=dyn_shldrCapHorn_c_r}
    constraintSetup [
     {targetBone {boneName=clavicle_r}}
     {targetBone {boneName=upperarm_twist_01_r}}
     {targetBone {boneName=upperarm_r}}
     {targetBone {boneName=spine_05}}
    ]
    constraintWeights [1,1,1,1]
   }
   n21 = "Constraint - dyn_shldrCapHorn_d_r (PPPP)"!(
    ComponentPose=@n20
    Alpha=1
    ConstraintWeights_0=0.6327
    ConstraintWeights_1:float=0.209
    ConstraintWeights_2:float=0
    ConstraintWeights_3:float=0.1583
   ) -> (Pose) {
    alphaBoolBlend {blendOption=Linear}
    boneToModify {boneName=dyn_shldrCapHorn_d_r}
    constraintSetup [
     {targetBone {boneName=clavicle_r}}
     {targetBone {boneName=upperarm_twist_01_r}}
     {targetBone {boneName=upperarm_r}}
     {targetBone {boneName=spine_05}}
    ]
    constraintWeights [1,1,1,1]
   }
   n22 = "Constraint - dyn_shldrCapHorn_e_r (PPPP)"!(
    ComponentPose=@n21
    Alpha=1
    ConstraintWeights_0=0.3559
    ConstraintWeights_1:float=0.6319
    ConstraintWeights_2:float=0
    ConstraintWeights_3:float=0.0122
   ) -> (Pose) {
    alphaBoolBlend {blendOption=Linear}
    boneToModify {boneName=dyn_shldrCapHorn_e_r}
    constraintSetup [
     {targetBone {boneName=clavicle_r}}
     {targetBone {boneName=upperarm_twist_01_r}}
     {targetBone {boneName=upperarm_r}}
     {targetBone {boneName=spine_05}}
    ]
    constraintWeights [1,1,1,1]
   }
   n23 = "Constraint - dyn_shldrCapHorn_f_r (PPPP)"!(
    ComponentPose=@n22
    Alpha=1
    ConstraintWeights_0=0.6995
    ConstraintWeights_1:float=0.1331
    ConstraintWeights_2:float=0
    ConstraintWeights_3:float=0.1673
   ) -> (Pose) {
    alphaBoolBlend {blendOption=Linear}
    boneToModify {boneName=dyn_shldrCapHorn_f_r}
    constraintSetup [
     {targetBone {boneName=clavicle_r}}
     {targetBone {boneName=upperarm_twist_01_r}}
     {targetBone {boneName=upperarm_r}}
     {targetBone {boneName=spine_05}}
    ]
    constraintWeights [1,1,1,1]
   }
   n24 = "Constraint - dyn_shldrCapHorn_h_r (PPPP)"!(
    ComponentPose=@n23
    Alpha=1
    ConstraintWeights_0=0.6547
    ConstraintWeights_1:float=0.2865
    ConstraintWeights_2:float=0
    ConstraintWeights_3:float=0.0587
   ) -> (Pose) {
    alphaBoolBlend {blendOption=Linear}
    boneToModify {boneName=dyn_shldrCapHorn_h_r}
    constraintSetup [
     {targetBone {boneName=clavicle_r}}
     {targetBone {boneName=upperarm_twist_01_r}}
     {targetBone {boneName=upperarm_r}}
     {targetBone {boneName=spine_05}}
    ]
    constraintWeights [1,1,1,1]
   }
   n25 = "Constraint - dyn_shldrCapHorn_g_r (PPPP)"!(
    ComponentPose=@n24
    Alpha=1
    ConstraintWeights_0=0.6112
    ConstraintWeights_1:float=0.2378
    ConstraintWeights_2:float=0
    ConstraintWeights_3:float=0.151
   ) -> (Pose) {
    alphaBoolBlend {blendOption=Linear}
    boneToModify {boneName=dyn_shldrCapHorn_g_r}
    constraintSetup [
     {targetBone {boneName=clavicle_r}}
     {targetBone {boneName=upperarm_twist_01_r}}
     {targetBone {boneName=upperarm_r}}
     {targetBone {boneName=spine_05}}
    ]
    constraintWeights [1,1,1,1]
   }
   n26 = "Transform (Modify) Bone - Bone: dyn_shldrCapHorn_g_r"(ComponentPose=@n25, Alpha=Get uparm_rot_r() -> (uparm_rot_r:double), Translation="1,0,-1", Rotation="-10,-10,0", Scale="1,0.6,0.6") {
    rotationMode=BMM_Additive
    rotationSpace=BCS_BoneSpace
    scaleMode=BMM_Additive
    scaleSpace=BCS_BoneSpace
    translationMode=BMM_Additive
    translationSpace=BCS_BoneSpace
    alphaBoolBlend {blendOption=Linear}
    alphaScaleBiasClamp {bMapRange=true, interpSpeedDecreasing=25, interpSpeedIncreasing=25, inRange {max=85}}
    boneToModify {boneName=dyn_shldrCapHorn_g_r}
    rotation {pitch=-10, yaw=-10}
    scale {y=0.6, z=0.6}
    translation {x=1, z=-1}
   }
   n27 = "Constraint - dyn_shoeString_l (PPPPP)"!(
    ComponentPose=@n26
    Alpha=1
    ConstraintWeights_0=0
    ConstraintWeights_1:float=0.4
    ConstraintWeights_2:float=0.6
    ConstraintWeights_3:float=0
    ConstraintWeights_4:float=0
   ) -> (Pose) {
    alphaBoolBlend {blendOption=Linear}
    boneToModify {boneName=dyn_shoeString_l}
    constraintSetup [
     {targetBone {boneName=calf_l}}
     {targetBone {boneName=calf_twist_01_l}}
     {targetBone {boneName=foot_l}}
     {targetBone {boneName=ball_l}}
     {targetBone {boneName=calf_twist_02_l}}
    ]
    constraintWeights [1,1,1,1,1]
   }
   n28 = "Constraint - dyn_shoeString_r (PPPPP)"!(
    ComponentPose=@n27
    Alpha=1
    ConstraintWeights_0=0
    ConstraintWeights_1:float=0
    ConstraintWeights_2:float=0.4
    ConstraintWeights_3:float=0.6
    ConstraintWeights_4:float=0
   ) -> (Pose) {
    alphaBoolBlend {blendOption=Linear}
    boneToModify {boneName=dyn_shoeString_r}
    constraintSetup [
     {targetBone {boneName=ball_r}}
     {targetBone {boneName=calf_twist_02_r}}
     {targetBone {boneName=calf_twist_01_r}}
     {targetBone {boneName=foot_r}}
     {targetBone {boneName=calf_r}}
    ]
    constraintWeights [1,1,1,1,1]
   }
   n29 = "Constraint - dyn_shirts_fr_base (PPPPP)"!(
    ComponentPose=@n28
    Alpha=1
    ConstraintWeights_0=0.0227
    ConstraintWeights_1:float=0.1714
    ConstraintWeights_2:float=0.1132
    ConstraintWeights_3:float=0.6763
    ConstraintWeights_4:float=0.0164
   ) -> (Pose) {
    alphaBoolBlend {blendOption=Linear}
    boneToModify {boneName=dyn_shirts_fr_base}
    constraintSetup [
     {targetBone {boneName=spine_01}}
     {targetBone {boneName=spine_02}}
     {targetBone {boneName=spine_03}}
     {targetBone {boneName=spine_04}}
     {targetBone {boneName=pelvis}}
    ]
    constraintWeights [1,1,1,1,1]
   }
   n30 = "Constraint - dyn_shirts_fr_base_l (PPPPP)"!(
    ComponentPose=@n29
    Alpha=1
    ConstraintWeights_0=0.024
    ConstraintWeights_1:float=0.1419
    ConstraintWeights_2:float=0.1024
    ConstraintWeights_3:float=0.7184
    ConstraintWeights_4:float=0.0134
   ) -> (Pose) {
    alphaBoolBlend {blendOption=Linear}
    boneToModify {boneName=dyn_shirts_fr_base_l}
    constraintSetup [
     {targetBone {boneName=spine_01}}
     {targetBone {boneName=spine_02}}
     {targetBone {boneName=spine_03}}
     {targetBone {boneName=spine_04}}
     {targetBone {boneName=pelvis}}
    ]
    constraintWeights [1,1,1,1,1]
   }
   n31 = "Constraint - dyn_shirts_fr_base_r (PPPPP)"!(
    ComponentPose=@n30
    Alpha=1
    ConstraintWeights_0=0.0269
    ConstraintWeights_1:float=0.1864
    ConstraintWeights_2:float=0.1187
    ConstraintWeights_3:float=0.6468
    ConstraintWeights_4:float=0.0213
   ) -> (Pose) {
    alphaBoolBlend {blendOption=Linear}
    boneToModify {boneName=dyn_shirts_fr_base_r}
    constraintSetup [
     {targetBone {boneName=spine_01}}
     {targetBone {boneName=spine_02}}
     {targetBone {boneName=spine_03}}
     {targetBone {boneName=spine_04}}
     {targetBone {boneName=pelvis}}
    ]
    constraintWeights [1,1,1,1,1]
   }
   n32 = "Constraint - dyn_shirts_bk_base_l (PPPPP)"!(
    ComponentPose=@n31
    Alpha=1
    ConstraintWeights_0=0.0707
    ConstraintWeights_1:float=0.3129
    ConstraintWeights_2:float=0.1659
    ConstraintWeights_3:float=0.4009
    ConstraintWeights_4:float=0.0497
   ) -> (Pose) {
    alphaBoolBlend {blendOption=Linear}
    boneToModify {boneName=dyn_shirts_bk_base_l}
    constraintSetup [
     {targetBone {boneName=spine_01}}
     {targetBone {boneName=spine_02}}
     {targetBone {boneName=spine_03}}
     {targetBone {boneName=spine_04}}
     {targetBone {boneName=clavicle_l}}
    ]
    constraintWeights [1,1,1,1,1]
   }
   n33 = "Constraint - dyn_shirts_bk_base (PPPPP)"!(
    ComponentPose=@n32
    Alpha=1
    ConstraintWeights_0=0.0794
    ConstraintWeights_1:float=0.2121
    ConstraintWeights_2:float=0.1554
    ConstraintWeights_3:float=0.3853
    ConstraintWeights_4:float=0.1678
   ) -> (Pose) {
    alphaBoolBlend {blendOption=Linear}
    boneToModify {boneName=dyn_shirts_bk_base}
    constraintSetup [
     {targetBone {boneName=clavicle_r}}
     {targetBone {boneName=spine_02}}
     {targetBone {boneName=spine_03}}
     {targetBone {boneName=spine_04}}
     {targetBone {boneName=spine_05}}
    ]
    constraintWeights [1,1,1,1,1]
   }
   n34 = "Transform (Modify) Bone - Bone: dyn_sword_base"(ComponentPose=@n33, Alpha=1, Translation="0,0,1", Rotation="0,0,0", Scale="1,1,1") {
    rotationMode=BMM_Additive
    rotationSpace=BCS_BoneSpace
    scaleMode=BMM_Additive
    scaleSpace=BCS_BoneSpace
    translationMode=BMM_Additive
    translationSpace=BCS_BoneSpace
    alphaBoolBlend {blendOption=Linear}
    alphaScaleBiasClamp {inRange {max=0, min=-100}, outRange {max=0, min=1}}
    boneToModify {boneName=dyn_sword_base}
    translation {z=1}
   }
   n35 = "Transform (Modify) Bone - Bone: dyn_sword_base"(ComponentPose=@n34, Alpha=Get thigh_rot_l(), Translation="0,0,2", Rotation="0,-90,0", Scale="1,1,1") {
    rotationMode=BMM_Additive
    rotationSpace=BCS_BoneSpace
    scaleMode=BMM_Additive
    scaleSpace=BCS_BoneSpace
    translationMode=BMM_Additive
    translationSpace=BCS_BoneSpace
    alphaBoolBlend {blendOption=Linear}
    alphaScaleBiasClamp {bMapRange=true, inRange {max=0, min=-100}, outRange {max=0, min=1}}
    boneToModify {boneName=dyn_sword_base}
    rotation {yaw=-90}
    translation {z=2}
   }
   n36 = "Constraint - dyn_armor_chest_bit_a_base (PP)"!(ComponentPose=@n35, Alpha=1, ConstraintWeights_0=0.7962, ConstraintWeights_1:float=0.2038) -> (Pose) {
    alphaBoolBlend {blendOption=Linear}
    boneToModify {boneName=dyn_armor_chest_bit_a_base}
    constraintSetup [{targetBone {boneName=spine_04}}, {targetBone {boneName=spine_05}}]
    constraintWeights [1,1]
   }
   n37 = "Constraint - dyn_armor_chest_bit_b_base (PPPPPPP)"!(
    ComponentPose=@n36
    Alpha=1
    ConstraintWeights_0=0.0032
    ConstraintWeights_1:float=0.0147
    ConstraintWeights_2:float=0.0161
    ConstraintWeights_3:float=0.8757
    ConstraintWeights_4:float=0.0873
    ConstraintWeights_5:float=0.0028
    ConstraintWeights_6:float=0.0001
   ) -> (Pose) {
    alphaBoolBlend {blendOption=Linear}
    boneToModify {boneName=dyn_armor_chest_bit_b_base}
    constraintSetup [
     {targetBone {boneName=clavicle_l}}
     {targetBone {boneName=spine_02}}
     {targetBone {boneName=spine_03}}
     {targetBone {boneName=spine_04}}
     {targetBone {boneName=spine_05}}
     {targetBone {boneName=clavicle_r}}
     {targetBone {boneName=pelvis}}
    ]
    constraintWeights [1,1,1,1,1,1,1]
   }
   n38 = "Constraint - dyn_armor_back_bit_base (PPPPPPP)"!(
    ComponentPose=@n37
    Alpha=1
    ConstraintWeights_0=0.3604
    ConstraintWeights_1:float=0.0221
    ConstraintWeights_2:float=0.0292
    ConstraintWeights_3:float=0.2439
    ConstraintWeights_4:float=0.3139
    ConstraintWeights_5:float=0.0305
    ConstraintWeights_6:float=0
   ) -> (Pose) {
    alphaBoolBlend {blendOption=Linear}
    boneToModify {boneName=dyn_armor_back_bit_base}
    constraintSetup [
     {targetBone {boneName=clavicle_l}}
     {targetBone {boneName=spine_02}}
     {targetBone {boneName=spine_03}}
     {targetBone {boneName=spine_04}}
     {targetBone {boneName=spine_05}}
     {targetBone {boneName=clavicle_r}}
     {targetBone {boneName=pelvis}}
    ]
    constraintWeights [1,1,1,1,1,1,1]
   }
   n39 = "Constraint - dyn_skirt_fr_a_base_l (RR)"!(ComponentPose=@n38, Alpha=1, ConstraintWeights_0=0, ConstraintWeights_1:float=1) -> (Pose) {
    alphaBoolBlend {blendOption=Linear}
    alphaScaleBiasClamp {bInterpResult=true}
    boneToModify {boneName=dyn_skirt_fr_a_base_l}
    constraintSetup [{transformType=Rotation, targetBone {boneName=pelvis}}, {transformType=Rotation, targetBone {boneName=thigh_twist_01_l}}]
    constraintWeights [1,1]
   }
   n40 = "Constraint - dyn_skirt_side_a_base_l (RR)"!(ComponentPose=@n39, Alpha=1, ConstraintWeights_0=0.3, ConstraintWeights_1:float=0.7) -> (Pose) {
    alphaBoolBlend {blendOption=Linear}
    boneToModify {boneName=dyn_skirt_side_a_base_l}
    constraintSetup [{transformType=Rotation, targetBone {boneName=pelvis}}, {transformType=Rotation, targetBone {boneName=thigh_twist_01_l}}]
    constraintWeights [1,1]
   }
   n41 = "Constraint - dyn_skirt_bk_a_base_l (RR)"!(ComponentPose=@n40, Alpha=1, ConstraintWeights_0=1, ConstraintWeights_1:float=0) -> (Pose) {
    alphaBoolBlend {blendOption=Linear}
    boneToModify {boneName=dyn_skirt_bk_a_base_l}
    constraintSetup [{transformType=Rotation, targetBone {boneName=pelvis}}, {transformType=Rotation, targetBone {boneName=thigh_twist_01_l}}]
    constraintWeights [1,1]
   }
   n42 = "Constraint - dyn_skirt_fr_a_base_r (RR)"!(ComponentPose=@n41, Alpha=1, ConstraintWeights_0=0, ConstraintWeights_1:float=1) -> (Pose) {
    alphaBoolBlend {blendOption=Linear}
    boneToModify {boneName=dyn_skirt_fr_a_base_r}
    constraintSetup [{transformType=Rotation, targetBone {boneName=pelvis}}, {transformType=Rotation, targetBone {boneName=thigh_twist_01_r}}]
    constraintWeights [1,1]
   }
   n43 = "Constraint - dyn_skirt_side_a_base_r (RR)"!(ComponentPose=@n42, Alpha=1, ConstraintWeights_0=0.3, ConstraintWeights_1:float=0.7) -> (Pose) {
    alphaBoolBlend {blendOption=Linear}
    boneToModify {boneName=dyn_skirt_side_a_base_r}
    constraintSetup [{transformType=Rotation, targetBone {boneName=pelvis}}, {transformType=Rotation, targetBone {boneName=thigh_twist_01_r}}]
    constraintWeights [1,1]
   }
   n44 = "Constraint - dyn_skirt_bk_a_base_r (RR)"!(ComponentPose=@n43, Alpha=1, ConstraintWeights_0=1, ConstraintWeights_1:float=0) -> (Pose) {
    alphaBoolBlend {blendOption=Linear}
    boneToModify {boneName=dyn_skirt_bk_a_base_r}
    constraintSetup [{transformType=Rotation, targetBone {boneName=pelvis}}, {transformType=Rotation, targetBone {boneName=thigh_twist_01_r}}]
    constraintWeights [1,1]
   }
   n45 = Get thigh_rot_l()
   n46 = "Transform (Modify) Bone - Bone: dyn_skirt_fr_a_base_l"(ComponentPose=@n44, Alpha=@n45, Translation="1,5,3", Rotation="5,0,0", Scale="1,1,1") {
    rotationMode=BMM_Additive
    rotationSpace=BCS_BoneSpace
    scaleMode=BMM_Additive
    scaleSpace=BCS_BoneSpace
    translationMode=BMM_Additive
    translationSpace=BCS_BoneSpace
    alphaBoolBlend {blendOption=Linear}
    alphaScaleBiasClamp {bMapRange=true, inRange {max=0, min=-100}, outRange {max=0, min=1}}
    boneToModify {boneName=dyn_skirt_fr_a_base_l}
    rotation {pitch=5}
    translation=<1 5 3>
   }
   n47 = "Transform (Modify) Bone - Bone: dyn_skirt_side_a_base_l"(ComponentPose=@n46, Alpha=@n45, Translation="-2,2,2", Rotation="0,0,0", Scale="1,1,1") {
    rotationMode=BMM_Additive
    rotationSpace=BCS_BoneSpace
    scaleMode=BMM_Additive
    scaleSpace=BCS_BoneSpace
    translationMode=BMM_Additive
    translationSpace=BCS_BoneSpace
    alphaBoolBlend {blendOption=Linear}
    alphaScaleBiasClamp {bMapRange=true, inRange {max=0, min=-100}, outRange {max=0, min=1}}
    boneToModify {boneName=dyn_skirt_side_a_base_l}
    translation=<-2 2 2>
   }
   n48 = "Transform (Modify) Bone - Bone: dyn_skirt_bk_a_base_l"(ComponentPose=@n47, Alpha=@n45, Translation="0.5,0,-2", Rotation="-5,0,0", Scale="1,1,1") {
    rotationMode=BMM_Additive
    rotationSpace=BCS_BoneSpace
    scaleMode=BMM_Additive
    scaleSpace=BCS_BoneSpace
    translationMode=BMM_Additive
    translationSpace=BCS_BoneSpace
    alphaBoolBlend {blendOption=Linear}
    alphaScaleBiasClamp {bMapRange=true, inRange {max=0, min=-100}, outRange {max=0, min=1}}
    boneToModify {boneName=dyn_skirt_bk_a_base_l}
    rotation {pitch=-5}
    translation {x=0.5, z=-2}
   }
   n49 = Get thigh_rot_r()
   n50 = "Transform (Modify) Bone - Bone: dyn_skirt_fr_a_base_r"(ComponentPose=@n48, Alpha=@n49, Translation="0,-5,-3", Rotation="5,10,0", Scale="1,1,1") {
    rotationMode=BMM_Additive
    rotationSpace=BCS_BoneSpace
    scaleMode=BMM_Additive
    scaleSpace=BCS_BoneSpace
    translationMode=BMM_Additive
    translationSpace=BCS_BoneSpace
    alphaBoolBlend {blendOption=Linear}
    alphaScaleBiasClamp {bMapRange=true, inRange {max=0, min=-100}, outRange {max=0, min=1}}
    boneToModify {boneName=dyn_skirt_fr_a_base_r}
    rotation {pitch=5, yaw=10}
    translation {y=-5, z=-3}
   }
   n51 = "Transform (Modify) Bone - Bone: dyn_skirt_side_a_base_r"(ComponentPose=@n50, Alpha=@n49, Translation="4,-2,-2", Rotation="0,0,0", Scale="1,1,1") {
    rotationMode=BMM_Additive
    rotationSpace=BCS_BoneSpace
    scaleMode=BMM_Additive
    scaleSpace=BCS_BoneSpace
    translationMode=BMM_Additive
    translationSpace=BCS_BoneSpace
    alphaBoolBlend {blendOption=Linear}
    alphaScaleBiasClamp {bMapRange=true, inRange {max=0, min=-100}, outRange {max=0, min=1}}
    boneToModify {boneName=dyn_skirt_side_a_base_r}
    translation=<4 -2 -2>
   }
   n52 = "Transform (Modify) Bone - Bone: dyn_skirt_bk_a_base_r"(ComponentPose=@n51, Alpha=@n49, Translation="-1,1,2", Rotation="-5,0,0", Scale="1,1,1") {
    rotationMode=BMM_Additive
    rotationSpace=BCS_BoneSpace
    scaleMode=BMM_Additive
    scaleSpace=BCS_BoneSpace
    translationMode=BMM_Additive
    translationSpace=BCS_BoneSpace
    alphaBoolBlend {blendOption=Linear}
    alphaScaleBiasClamp {bMapRange=true, inRange {max=0, min=-100}, outRange {max=0, min=1}}
    boneToModify {boneName=dyn_skirt_bk_a_base_r}
    rotation {pitch=-5}
    translation=<-1 1 2>
   }
   n53 = "Transform (Modify) Bone - Bone: dyn_armor_belt_fr_base_r"(ComponentPose=@n52, Alpha=1, Translation="2,0,0", Rotation="0,0,0", Scale="1,1,1") {
    rotationMode=BMM_Additive
    rotationSpace=BCS_BoneSpace
    scaleMode=BMM_Additive
    scaleSpace=BCS_BoneSpace
    translationMode=BMM_Additive
    translationSpace=BCS_BoneSpace
    alphaBoolBlend {blendOption=Linear}
    alphaScaleBiasClamp {inRange {max=0, min=-100}, outRange {max=0, min=1}}
    boneToModify {boneName=dyn_armor_belt_fr_base_r}
    translation {x=2}
   }
   n54 = Get thigh_rot_l()
   n55 = "Transform (Modify) Bone - Bone: dyn_armor_belt_fr_base_l"(ComponentPose=@n53, Alpha=@n54, Translation="-3,0,0", Rotation="0,0,0", Scale="1,1,1") {
    rotationMode=BMM_Additive
    rotationSpace=BCS_BoneSpace
    scaleMode=BMM_Additive
    scaleSpace=BCS_BoneSpace
    translationMode=BMM_Additive
    translationSpace=BCS_BoneSpace
    alphaBoolBlend {blendOption=Linear}
    alphaScaleBiasClamp {bMapRange=true, inRange {max=0, min=-100}, outRange {max=0, min=1}}
    boneToModify {boneName=dyn_armor_belt_fr_base_l}
    translation {x=-3}
   }
   n56 = "Transform (Modify) Bone - Bone: dyn_armor_belt_fr_base_r"(ComponentPose=@n55, Alpha=@n54, Translation="2,0,0", Rotation="0,0,0", Scale="1,1,1") {
    rotationMode=BMM_Additive
    rotationSpace=BCS_BoneSpace
    scaleMode=BMM_Additive
    scaleSpace=BCS_BoneSpace
    translationMode=BMM_Additive
    translationSpace=BCS_BoneSpace
    alphaBoolBlend {blendOption=Linear}
    alphaScaleBiasClamp {bMapRange=true, inRange {max=0, min=-100}, outRange {max=0, min=1}}
    boneToModify {boneName=dyn_armor_belt_fr_base_r}
    translation {x=2}
   }
   n57 = Get thigh_rot_r()
   n58 = "Transform (Modify) Bone - Bone: dyn_armor_belt_fr_base_r"(ComponentPose=@n56, Alpha=@n57, Translation="5,0,0", Rotation="0,0,0", Scale="1,1,1") {
    rotationMode=BMM_Additive
    rotationSpace=BCS_BoneSpace
    scaleMode=BMM_Additive
    scaleSpace=BCS_BoneSpace
    translationMode=BMM_Additive
    translationSpace=BCS_BoneSpace
    alphaBoolBlend {blendOption=Linear}
    alphaScaleBiasClamp {bMapRange=true, inRange {max=0, min=-100}, outRange {max=0, min=1}}
    boneToModify {boneName=dyn_armor_belt_fr_base_r}
    translation {x=5}
   }
   n59 = "Transform (Modify) Bone - Bone: dyn_armor_belt_side_base_r"(ComponentPose=@n58, Alpha=@n57, Translation="3,0,0", Rotation="0,0,0", Scale="1,1,1") {
    rotationMode=BMM_Additive
    rotationSpace=BCS_BoneSpace
    scaleMode=BMM_Additive
    scaleSpace=BCS_BoneSpace
    translationMode=BMM_Additive
    translationSpace=BCS_BoneSpace
    alphaBoolBlend {blendOption=Linear}
    alphaScaleBiasClamp {bMapRange=true, inRange {max=0, min=-100}, outRange {max=0, min=1}}
    boneToModify {boneName=dyn_armor_belt_side_base_r}
    translation {x=3}
   }
   n60 = "Constraint - dyn_armor_rope_hook_base (PPPPPP)"!(
    ComponentPose=@n59
    Alpha=1
    ConstraintWeights_0=0.5213
    ConstraintWeights_1:float=0.015
    ConstraintWeights_2:float=0.303
    ConstraintWeights_3:float=0.02
    ConstraintWeights_4:float=0.0988
    ConstraintWeights_5:float=0.0417
   ) -> (Pose) {
    alphaBoolBlend {blendOption=Linear}
    boneToModify {boneName=dyn_armor_rope_hook_base}
    constraintSetup [
     {targetBone {boneName=dyn_skirt_side_a_r}}
     {targetBone {boneName=dyn_skirt_bk_b_r}}
     {targetBone {boneName=dyn_skirt_fr_a_r}}
     {targetBone {boneName=dyn_skirt_fr_b_r}}
     {targetBone {boneName=dyn_skirt_bk_a_r}}
     {targetBone {boneName=dyn_skirt_side_b_r}}
    ]
    constraintWeights [1,1,1,1,1,1]
   }
   n61 = "Constraint - dyn_armor_rope_a (PPPPP)"!(
    ComponentPose=@n60
    Alpha=1
    ConstraintWeights_0=0.4439
    ConstraintWeights_1:float=0.4354
    ConstraintWeights_2:float=0.0245
    ConstraintWeights_3:float=0.0601
    ConstraintWeights_4:float=0.0362
   ) -> (Pose) {
    alphaBoolBlend {blendOption=Linear}
    boneToModify {boneName=dyn_armor_rope_a}
    constraintSetup [
     {targetBone {boneName=dyn_skirt_side_a_r}}
     {targetBone {boneName=dyn_skirt_fr_a_r}}
     {targetBone {boneName=dyn_skirt_fr_b_r}}
     {targetBone {boneName=dyn_skirt_bk_a_r}}
     {targetBone {boneName=dyn_skirt_side_b_r}}
    ]
    constraintWeights [1,1,1,1,1]
   }
   n62 = "Constraint - dyn_armor_rope_b (PPPPPP)"!(
    ComponentPose=@n61
    Alpha=1
    ConstraintWeights_0=0.3319
    ConstraintWeights_1:float=0.0482
    ConstraintWeights_2:float=0.1491
    ConstraintWeights_3:float=0.109
    ConstraintWeights_4:float=0.0466
    ConstraintWeights_5:float=0.3152
   ) -> (Pose) {
    alphaBoolBlend {blendOption=Linear}
    boneToModify {boneName=dyn_armor_rope_b}
    constraintSetup [
     {targetBone {boneName=dyn_skirt_side_a_r}}
     {targetBone {boneName=dyn_skirt_bk_b_r}}
     {targetBone {boneName=dyn_skirt_fr_a_r}}
     {targetBone {boneName=dyn_skirt_fr_b_r}}
     {targetBone {boneName=dyn_skirt_bk_a_r}}
     {targetBone {boneName=dyn_skirt_side_b_r}}
    ]
    constraintWeights [1,1,1,1,1,1]
   }
   n63 = "Transform (Modify) Bone - Bone: dyn_armor_rope_hook_base"(ComponentPose=@n62, Alpha=1, Translation="0,1,0", Rotation="0,0,0", Scale="1,1,1") {
    rotationMode=BMM_Additive
    rotationSpace=BCS_BoneSpace
    scaleMode=BMM_Additive
    scaleSpace=BCS_BoneSpace
    translationMode=BMM_Additive
    translationSpace=BCS_BoneSpace
    alphaBoolBlend {blendOption=Linear}
    alphaScaleBiasClamp {inRange {max=0, min=-100}, outRange {max=0, min=1}}
    boneToModify {boneName=dyn_armor_rope_hook_base}
    translation {y=1}
   }
   n64 = RigidBody(ComponentPose=@n63) {
    bClampLinearTranslationLimitToRefPose=true
    bOverrideWorldGravity=true
    overridePhysicsAsset=&/F_MED_Dino_Hunter_Armor_Shirts_Physics
    simulationSpace=BaseBoneSpace
    alphaBoolBlend {blendOption=Linear}
    baseBoneRef {boneName=root}
    componentLinearAccScale=<0.35 0.35 0.35>
    componentLinearVelScale=<0.35 0.35 0.35>
    overrideWorldGravity {z=-980}
   }
   n65 = RigidBody(ComponentPose=@n64) {
    bClampLinearTranslationLimitToRefPose=true
    bOverrideWorldGravity=true
    overridePhysicsAsset=&/F_MED_Dino_Hunter_Armor_Physics
    simulationSpace=BaseBoneSpace
    alphaBoolBlend {blendOption=Linear}
    baseBoneRef {boneName=root}
    componentLinearAccScale=<0.35 0.35 0.35>
    componentLinearVelScale=<0.35 0.35 0.35>
    overrideWorldGravity {z=-980}
   }
   n66 = RigidBody(ComponentPose=@n65) {
    bClampLinearTranslationLimitToRefPose=true
    bOverrideWorldGravity=true
    overridePhysicsAsset=&/F_MED_Dino_Hunter_Armor_Rope_Physics
    simulationSpace=BaseBoneSpace
    alphaBoolBlend {blendOption=Linear}
    baseBoneRef {boneName=root}
    componentLinearAccScale=<0.35 0.35 0.35>
    componentLinearVelScale=<0.35 0.35 0.35>
    overrideWorldGravity {z=-980}
   }
   n67 = Anim Dynamics - Bone: dyn_swordTeeth_a(ComponentPose=@n66, Alpha=1, ExternalForce="0,0,0") {
    angularDampingOverride=1
    angularSpringConstant=1500
    bAngularSpring=true
    bGravityOverrideInSimSpace=true
    bOverrideAngularDamping=true
    gravityScale=0.35
    simulationSpace=RootRelative
    alphaBoolBlend {blendOption=Linear}
    boundBone {boneName=dyn_swordTeeth_a}
    physicsBodyDefinitions [{
      sphereCollisionRadius=0
      boundBone {boneName=dyn_swordTeeth_a}
      boxExtents=<20 10 20>
      constraintSetup {twistAxis=AxisY, angularLimitsMax=<25 0 15>, angularLimitsMin=<-25 0 -15>, angularTarget=<0 0 0>}
      localJointOffset=<0 10 0>
     }]
   }
   n68 = Anim Dynamics - Bone: dyn_swordTeeth_b(ComponentPose=@n67, Alpha=0.75, ExternalForce="0,0,0") {
    angularDampingOverride=1
    angularSpringConstant=1500
    bAngularSpring=true
    bGravityOverrideInSimSpace=true
    bOverrideAngularDamping=true
    gravityScale=0.35
    simulationSpace=RootRelative
    alphaBoolBlend {blendOption=Linear}
    boundBone {boneName=dyn_swordTeeth_b}
    physicsBodyDefinitions [{
      sphereCollisionRadius=0
      boundBone {boneName=dyn_swordTeeth_b}
      boxExtents=<20 10 20>
      constraintSetup {twistAxis=AxisY, angularLimitsMax=<20 0 15>, angularLimitsMin=<-20 0 -15>, angularTarget=<0 0 0>}
      localJointOffset=<0 10 0>
     }]
   }
   n69 = Anim Dynamics - Bone: dyn_swordTeeth_c(ComponentPose=@n68, Alpha=0.5, ExternalForce="0,0,0") {
    angularDampingOverride=1
    angularSpringConstant=1500
    bAngularSpring=true
    bGravityOverrideInSimSpace=true
    bOverrideAngularDamping=true
    gravityScale=0.35
    simulationSpace=RootRelative
    alphaBoolBlend {blendOption=Linear}
    boundBone {boneName=dyn_swordTeeth_c}
    physicsBodyDefinitions [{
      sphereCollisionRadius=0
      boundBone {boneName=dyn_swordTeeth_c}
      boxExtents=<20 10 20>
      constraintSetup {twistAxis=AxisY, angularLimitsMax=<15 0 15>, angularLimitsMin=<-15 0 -15>, angularTarget=<0 0 0>}
      localJointOffset=<0 10 0>
     }]
   }
   n70 = Anim Dynamics - Bone: dyn_shoeString_a_l(ComponentPose=@n69, Alpha=1, ExternalForce="0,0,0") {
    angularDampingOverride=1
    angularSpringConstant=1500
    bAngularSpring=true
    bGravityOverrideInSimSpace=true
    bOverrideAngularDamping=true
    gravityScale=0.35
    simulationSpace=RootRelative
    alphaBoolBlend {blendOption=Linear}
    boundBone {boneName=dyn_shoeString_a_l}
    physicsBodyDefinitions [{
      sphereCollisionRadius=0
      boundBone {boneName=dyn_shoeString_a_l}
      boxExtents=<10 20 20>
      constraintSetup {angularLimitsMax=<10 0 0>, angularLimitsMin=<-10 -25 0>, angularTarget=<0 0 0>}
      localJointOffset=<10 0 0>
     }]
   }
   n71 = Anim Dynamics - Bone: dyn_shoeString_b_l(ComponentPose=@n70, Alpha=1, ExternalForce="0,0,0") {
    angularDampingOverride=1
    angularSpringConstant=1500
    bAngularSpring=true
    bGravityOverrideInSimSpace=true
    bOverrideAngularDamping=true
    gravityScale=0.35
    simulationSpace=RootRelative
    alphaBoolBlend {blendOption=Linear}
    boundBone {boneName=dyn_shoeString_b_l}
    physicsBodyDefinitions [{
      sphereCollisionRadius=0
      boundBone {boneName=dyn_shoeString_b_l}
      boxExtents=<10 20 20>
      constraintSetup {angularLimitsMax=<10 25 0>, angularLimitsMin=<-10 0 0>, angularTarget=<0 0 0>}
      localJointOffset=<10 0 0>
     }]
   }
   n72 = Anim Dynamics - Bone: dyn_shoeString_b_r(ComponentPose=@n71, Alpha=1, ExternalForce="0,0,0") {
    angularDampingOverride=1
    angularSpringConstant=1500
    bAngularSpring=true
    bGravityOverrideInSimSpace=true
    bOverrideAngularDamping=true
    gravityScale=0.35
    simulationSpace=RootRelative
    alphaBoolBlend {blendOption=Linear}
    boundBone {boneName=dyn_shoeString_b_r}
    physicsBodyDefinitions [{
      sphereCollisionRadius=0
      boundBone {boneName=dyn_shoeString_b_r}
      boxExtents=<10 20 20>
      constraintSetup {angularLimitsMax=<10 0 0>, angularLimitsMin=<-10 -25 0>, angularTarget=<0 0 0>}
      localJointOffset=<-10 0 0>
     }]
   }
   n73 = Anim Dynamics - Bone: dyn_shoeString_a_r(ComponentPose=@n72, Alpha=1, ExternalForce="0,0,0") {
    angularDampingOverride=1
    angularSpringConstant=1500
    bAngularSpring=true
    bGravityOverrideInSimSpace=true
    bOverrideAngularDamping=true
    gravityScale=0.35
    simulationSpace=RootRelative
    alphaBoolBlend {blendOption=Linear}
    boundBone {boneName=dyn_shoeString_a_r}
    physicsBodyDefinitions [{
      sphereCollisionRadius=0
      boundBone {boneName=dyn_shoeString_a_r}
      boxExtents=<10 20 20>
      constraintSetup {angularLimitsMax=<10 0 0>, angularLimitsMin=<-10 -25 0>, angularTarget=<0 0 0>}
      localJointOffset=<-10 0 0>
     }]
   }
   n74 = Anim Dynamics - Bone: dyn_armor_chest_bit_a_base(ComponentPose=@n73, Alpha=0.5, ExternalForce="0,0,0") {
    angularDampingOverride=1
    angularSpringConstant=1500
    bAngularSpring=true
    bGravityOverrideInSimSpace=true
    bOverrideAngularDamping=true
    gravityScale=0.35
    simulationSpace=RootRelative
    alphaBoolBlend {blendOption=Linear}
    boundBone {boneName=dyn_armor_chest_bit_a_base}
    physicsBodyDefinitions [{
      sphereCollisionRadius=0
      boundBone {boneName=dyn_armor_chest_bit_a_base}
      boxExtents=<10 20 20>
      constraintSetup {angularLimitsMax=<0 5 15>, angularLimitsMin=<0 -5 -15>, angularTarget=<0 0 0>}
      localJointOffset=<10 0 0>
     }]
   }
   n75 = Anim Dynamics - Bone: dyn_armor_chest_bit_b_base(ComponentPose=@n74, Alpha=1, ExternalForce="0,0,0") {
    angularDampingOverride=1
    angularSpringConstant=1500
    bAngularSpring=true
    bGravityOverrideInSimSpace=true
    bOverrideAngularDamping=true
    gravityScale=0.35
    simulationSpace=RootRelative
    alphaBoolBlend {blendOption=Linear}
    boundBone {boneName=dyn_armor_chest_bit_b_base}
    physicsBodyDefinitions [{
      sphereCollisionRadius=0
      boundBone {boneName=dyn_armor_chest_bit_b_base}
      boxExtents=<10 20 20>
      constraintSetup {angularLimitsMax=<5 10 30>, angularLimitsMin=<-5 -10 -10>, angularTarget=<0 0 0>}
      localJointOffset=<10 0 0>
     }]
   }
   n76 = Anim Dynamics - Bone: dyn_armor_back_bit_base(ComponentPose=@n75, Alpha=0.9, ExternalForce="0,0,0") {
    angularDampingOverride=1
    angularSpringConstant=1500
    bAngularSpring=true
    bGravityOverrideInSimSpace=true
    bOverrideAngularDamping=true
    gravityScale=0.35
    simulationSpace=RootRelative
    alphaBoolBlend {blendOption=Linear}
    boundBone {boneName=dyn_armor_back_bit_base}
    componentLinearAccScale=<0.35 0.35 0.35>
    componentLinearVelScale=<0.35 0.35 0.35>
    physicsBodyDefinitions [{
      sphereCollisionRadius=0
      boundBone {boneName=dyn_armor_back_bit_base}
      boxExtents=<10 20 20>
      constraintSetup {angularLimitsMax=<5 5 10>, angularLimitsMin=<-5 -5 -60>, angularTarget=<0 0 0>}
      localJointOffset=<10 0 0>
     }]
   }
   Output Pose(Result=Component To Local(ComponentPose=@n76)) {layerGroup=None}
   Reroute Node(InputPin:Wildcard) -> (OutputPin:Wildcard)
   n77 = "Transform (Modify) Bone - Bone: dyn_skirt_side_a_base_l"(Alpha=Get thigh_side_l() -> (thigh_side_l:double), Translation="0,2,0", Rotation="0,0,0", Scale="1,1,1") {
    rotationMode=BMM_Additive
    rotationSpace=BCS_BoneSpace
    scaleMode=BMM_Additive
    scaleSpace=BCS_BoneSpace
    translationMode=BMM_Additive
    translationSpace=BCS_BoneSpace
    alphaBoolBlend {blendOption=Linear}
    alphaScaleBiasClamp {bMapRange=true, inRange {max=0, min=-5}, outRange {max=0, min=1}}
    boneToModify {boneName=dyn_skirt_side_a_base_l}
    translation {y=2}
   }
   n78 = Get thigh_side_r() -> (thigh_side_r:double)
   n79 = "Transform (Modify) Bone - Bone: dyn_skirt_side_a_base_r"(ComponentPose=@n77, Alpha=@n78, Translation="0,-2,0", Rotation="0,0,0", Scale="1,1,1") {
    rotationMode=BMM_Additive
    rotationSpace=BCS_BoneSpace
    scaleMode=BMM_Additive
    scaleSpace=BCS_BoneSpace
    translationMode=BMM_Additive
    translationSpace=BCS_BoneSpace
    alphaBoolBlend {blendOption=Linear}
    alphaScaleBiasClamp {bMapRange=true, inRange {max=0, min=-10}, outRange {max=0, min=1}}
    boneToModify {boneName=dyn_skirt_side_a_base_r}
    translation {y=-2}
   }
   n80 = "Transform (Modify) Bone - Bone: dyn_skirt_side_a_base_r"(ComponentPose=@n79, Alpha=@n78, Translation="0,0,0", Rotation="0,5,0", Scale="1,1,1") {
    rotationMode=BMM_Additive
    rotationSpace=BCS_BoneSpace
    scaleMode=BMM_Additive
    scaleSpace=BCS_BoneSpace
    translationMode=BMM_Additive
    translationSpace=BCS_BoneSpace
    alphaBoolBlend {blendOption=Linear}
    alphaScaleBiasClamp {bMapRange=true, inRange {max=20}}
    boneToModify {boneName=dyn_skirt_side_a_base_r}
    rotation {yaw=5}
   }
   n81 = "Transform (Modify) Bone - Bone: dyn_skirt_bk_a_base_r"(ComponentPose=@n80, Alpha=@n78, Translation="0,-3,3", Rotation="0,0,0", Scale="1,1,1") {
    rotationMode=BMM_Additive
    rotationSpace=BCS_BoneSpace
    scaleMode=BMM_Additive
    scaleSpace=BCS_BoneSpace
    translationMode=BMM_Additive
    translationSpace=BCS_BoneSpace
    alphaBoolBlend {blendOption=Linear}
    alphaScaleBiasClamp {bMapRange=true, inRange {max=0, min=-10}, outRange {max=0, min=1}}
    boneToModify {boneName=dyn_skirt_bk_a_base_r}
    translation {y=-3, z=3}
   }
   Reroute Node(InputPin:Component Space Pose Link Structure=@n81) -> (OutputPin:Component Space Pose Link Structure)
  ]}
 EventGraph {nodes [
   Event BlueprintUpdateAnimation()
   > n1 = Sequence!(execute) -> (then_0, then_1, then_2:Exec, then_3:Exec, then_4:Exec, then_5:Exec, then_6:Exec, then_7:Exec)
   n2 = Sequence!(execute=@n1.then_4) -> (then_0, then_1, then_2, then_3)
   n3 = GetOwnerDirection!(self:Custom Character Part Anim Instance Object Reference) -> (ReturnValue:float)
   n4 = OR Boolean(
    A=AND Boolean(
     A="float > float"(A=@n3, B=-120) {is=title K2Node_CallFunction}
     B="float < float"!(A:double=@n3, B:double=-60) -> (ReturnValue:bool) {is=title K2Node_CallFunction}
    )
    B=AND Boolean(A="float > float"(A=@n3, B=60), B="float < float"(A=@n3, B=120))
   )
   > n5 = Branch(Condition=@n4)
   n6 = GetOwnerSkeletalMeshComponent()
   n7 = GetDeltaTransformFromRefPose(self=@n6, BoneName=thigh_l, BaseName=pelvis)
   > Set thigh_side_l(thigh_side_l:double=BreakRotator(InRot=BreakTransform(InTransform=@n7).Rotation).Pitch) -> (Output_Get:double)
   Set thigh_side_l(execute=@n5.else, thigh_side_l=0)
   n8 = Branch(execute=@n2.then_1, Condition=@n4)
   n9 = GetDeltaTransformFromRefPose(self=@n6, BoneName=thigh_r, BaseName=pelvis)
   > Set thigh_side_r(thigh_side_r:double=BreakRotator(InRot=BreakTransform(InTransform=@n9).Rotation).Pitch) -> (Output_Get:double)
   Set thigh_side_r(execute=@n8.else, thigh_side_r=0)
   n10 = GetDeltaTransformFromRefPose(self=GetOwnerSkeletalMeshComponent(), BoneName=thigh_l, BaseName=pelvis)
   Set thigh_rot_l(execute=@n1.then_1, thigh_rot_l=BreakRotator(InRot=BreakTransform(InTransform=@n10).Rotation).Yaw)
   n11 = GetDeltaTransformFromRefPose(self=GetOwnerSkeletalMeshComponent(), BoneName=thigh_r, BaseName=pelvis)
   > Set thigh_rot_r(thigh_rot_r=BreakRotator(InRot=BreakTransform(InTransform=@n11).Rotation).Yaw)
   n12 = GetDeltaTransformFromRefPose(self=GetOwnerSkeletalMeshComponent(), BoneName=lowerarm_l, BaseName=upperarm_l)
   Set elbow_rot_l(execute=@n1.then_0, elbow_rot_l:double=BreakRotator(InRot=BreakTransform(InTransform=@n12).Rotation).Yaw) -> (Output_Get:double)
   n13 = GetDeltaTransformFromRefPose(self=GetOwnerSkeletalMeshComponent(), BoneName=lowerarm_r, BaseName=upperarm_r)
   > Set elbow_rot_r(elbow_rot_r:double=BreakRotator(InRot=BreakTransform(InTransform=@n13).Rotation).Yaw) -> (Output_Get:double)
   n14 = GetDeltaTransformFromRefPose(self=GetOwnerSkeletalMeshComponent(), BoneName=upperarm_l, BaseName=clavicle_l)
   Set uparm_rot_l(execute=@n1.then_3, uparm_rot_l:double=BreakRotator(InRot=BreakTransform(InTransform=@n14).Rotation).Yaw) -> (Output_Get:double)
   n15 = GetDeltaTransformFromRefPose(self=GetOwnerSkeletalMeshComponent(), BoneName=upperarm_r, BaseName=clavicle_r)
   > Set uparm_rot_r(uparm_rot_r:double=BreakRotator(InRot=BreakTransform(InTransform=@n15).Rotation).Yaw) -> (Output_Get:double)
   n16 = GetDeltaTransformFromRefPose(self=GetOwnerSkeletalMeshComponent(), BoneName=calf_l, BaseName=thigh_l)
   Set knee_rot_l(execute=@n1.then_2, knee_rot_l:double=BreakRotator(InRot=BreakTransform(InTransform=@n16).Rotation).Yaw) -> (Output_Get:double)
   n17 = GetDeltaTransformFromRefPose(self=GetOwnerSkeletalMeshComponent(), BoneName=calf_r, BaseName=thigh_r)
   > Set knee_rot_r(knee_rot_r:double=BreakRotator(InRot=BreakTransform(InTransform=@n17).Rotation).Yaw) -> (Output_Get:double)
   TryGetPawnOwner()
  ]}
]
variables [
 elbow_rot_l {type=number}
 elbow_rot_r {type=number}
 knee_rot_l {type=number}
 knee_rot_r {type=number}
 thigh_bk_pelvis_wt_l {type=number}
 thigh_bk_thigh_wt_l {type=number}
 thigh_fr_pelvis_wt_l {type=number}
 thigh_fr_thigh_wt_l {type=number}
 thigh_rot_l {type=number}
 thigh_rot_r {type=number}
 thigh_side_l {type=number}
 thigh_side_r {type=number}
 uparm_rot_l {type=number}
 uparm_rot_r {type=number}
]
