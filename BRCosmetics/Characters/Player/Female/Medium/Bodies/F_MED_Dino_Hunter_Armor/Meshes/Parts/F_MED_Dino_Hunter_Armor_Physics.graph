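# Physics asset for the female medium Dino Hunter Armor player cosmetic, covering the feather bones attached to the head.
# A small head capsule anchors feather chains (up, side, down and inner, with center, left and right variants), each a pair of capsule bodies.
# Each chain's first bone constrains to the head and its second to the first, with tight equal swing and twist limits, slightly looser on the inner chain.
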PhysicsAsset /BRCosmetics/Characters/Player/Female/Medium/Bodies/F_MED_Dino_Hunter_Armor/Meshes/Parts/F_MED_Dino_Hunter_Armor_Physics
bodies {
 dyn_feather_dn_a [{shapeType=Capsule, length=4, radius=1.75, center=<3 0 0>, rotation=<rot 90 0 0>}]
 dyn_feather_dn_a_l [{shapeType=Capsule, length=4, radius=1.75, center=<3 0 0>, rotation=<rot 90 0 0>}]
 dyn_feather_dn_a_r [{shapeType=Capsule, length=4, radius=1.75, center=<-3 0 0>, rotation=<rot 90 0 0>}]
 dyn_feather_dn_b [{shapeType=Capsule, length=4, radius=1.75, center=<3 0 0>, rotation=<rot 90 0 0>}]
 dyn_feather_dn_b_l [{shapeType=Capsule, length=4, radius=1.75, center=<3 0 0>, rotation=<rot 90 0 0>}]
 dyn_feather_dn_b_r [{shapeType=Capsule, length=4, radius=1.75, center=<-3 0 0>, rotation=<rot 90 0 0>}]
 dyn_feather_in_a [{shapeType=Capsule, length=4, radius=1.75, center=<3 0 0>, rotation=<rot 90 0 0>}]
 dyn_feather_in_b [{shapeType=Capsule, length=4, radius=1.75, center=<3 0 0>, rotation=<rot 90 0 0>}]
 dyn_feather_side_a_l [{shapeType=Capsule, length=4, radius=1.75, center=<3 0 0>, rotation=<rot 90 0 0>}]
 dyn_feather_side_a_r [{shapeType=Capsule, length=4, radius=1.75, center=<-3 0 0>, rotation=<rot 90 0 0>}]
 dyn_feather_side_b_l [{shapeType=Capsule, length=4, radius=1.75, center=<3 0 0>, rotation=<rot 90 0 0>}]
 dyn_feather_side_b_r [{shapeType=Capsule, length=4, radius=1.75, center=<-3 0 0>, rotation=<rot 90 0 0>}]
 dyn_feather_up_a [{shapeType=Capsule, length=4, radius=1.75, center=<3 0 0>, rotation=<rot 90 0 0>}]
 dyn_feather_up_a_l [{shapeType=Capsule, length=4, radius=1.75, center=<3 0 0>, rotation=<rot 90 0 0>}]
 dyn_feather_up_a_r [{shapeType=Capsule, length=4, radius=1.75, center=<-3 0 0>, rotation=<rot 90 0 0>}]
 dyn_feather_up_b [{shapeType=Capsule, length=4, radius=1.75, center=<3 0 0>, rotation=<rot 90 0 0>}]
 dyn_feather_up_b_l [{shapeType=Capsule, length=4, radius=1.75, center=<3 0 0>, rotation=<rot 90 0 0>}]
 dyn_feather_up_b_r [{shapeType=Capsule, length=4, radius=1.75, center=<-3 0 0>, rotation=<rot 90 0 0>}]
 head [{shapeType=Capsule, length=1, radius=1}]
}
constraints [
 {bone1Name=dyn_feather_up_a, bone2Name=head, swing1LimitDegrees=5, swing2LimitDegrees=5, twistLimitDegrees=5}
 {bone1Name=dyn_feather_up_b, bone2Name=dyn_feather_up_a, swing1LimitDegrees=5, swing2LimitDegrees=5, twistLimitDegrees=5}
 {bone1Name=dyn_feather_up_a_l, bone2Name=head, swing1LimitDegrees=5, swing2LimitDegrees=5, twistLimitDegrees=5}
 {bone1Name=dyn_feather_up_b_l, bone2Name=dyn_feather_up_a_l, swing1LimitDegrees=5, swing2LimitDegrees=5, twistLimitDegrees=5}
 {bone1Name=dyn_feather_side_a_l, bone2Name=head, swing1LimitDegrees=5, swing2LimitDegrees=5, twistLimitDegrees=5}
 {bone1Name=dyn_feather_side_b_l, bone2Name=dyn_feather_side_a_l, swing1LimitDegrees=5, swing2LimitDegrees=5, twistLimitDegrees=5}
 {bone1Name=dyn_feather_dn_a_l, bone2Name=head, swing1LimitDegrees=5, swing2LimitDegrees=5, twistLimitDegrees=5}
 {bone1Name=dyn_feather_dn_b_l, bone2Name=dyn_feather_dn_a_l, swing1LimitDegrees=5, swing2LimitDegrees=5, twistLimitDegrees=5}
 {bone1Name=dyn_feather_dn_a, bone2Name=head, swing1LimitDegrees=5, swing2LimitDegrees=5, twistLimitDegrees=5}
 {bone1Name=dyn_feather_dn_b, bone2Name=dyn_feather_dn_a, swing1LimitDegrees=5, swing2LimitDegrees=5, twistLimitDegrees=5}
 {bone1Name=dyn_feather_up_a_r, bone2Name=head, swing1LimitDegrees=5, swing2LimitDegrees=5, twistLimitDegrees=5}
 {bone1Name=dyn_feather_up_b_r, bone2Name=dyn_feather_up_a_r, swing1LimitDegrees=5, swing2LimitDegrees=5, twistLimitDegrees=5}
 {bone1Name=dyn_feather_side_a_r, bone2Name=head, swing1LimitDegrees=5, swing2LimitDegrees=5, twistLimitDegrees=5}
 {bone1Name=dyn_feather_side_b_r, bone2Name=dyn_feather_side_a_r, swing1LimitDegrees=5, swing2LimitDegrees=5, twistLimitDegrees=5}
 {bone1Name=dyn_feather_dn_a_r, bone2Name=head, swing1LimitDegrees=5, swing2LimitDegrees=5, twistLimitDegrees=5}
 {bone1Name=dyn_feather_dn_b_r, bone2Name=dyn_feather_dn_a_r, swing1LimitDegrees=5, swing2LimitDegrees=5, twistLimitDegrees=5}
 {bone1Name=dyn_feather_in_a, bone2Name=head, swing1LimitDegrees=10, swing2LimitDegrees=10, twistLimitDegrees=10}
 {bone1Name=dyn_feather_in_b, bone2Name=dyn_feather_in_a, swing1LimitDegrees=10, swing2LimitDegrees=10, twistLimitDegrees=10}
]
